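# Physics asset for the AshenMagnus gadget prop, providing simplified collision bodies for its skeletal mesh.
# A large, almost spherical capsule covers the main bone, and two smaller matching capsules cover the JNT_A_Target and JNT_B_Target bones.
# Each target body is constrained to main with identical moderate swing and twist limits, forming a simple two-branch hierarchy.
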PhysicsAsset /BRCosmetics/Animation/Props/Gadgets_AshenMagnus/Meshes/AshenMagnus_Gadget_Physics
bodies {
 JNT_A_Target [{shapeType=Capsule, length=52.0862, radius=52.0862, center=<0 1.2901 0>, rotation=<rot 0.0017 -90 0.0432>}]
 JNT_B_Target [{shapeType=Capsule, length=52.0864, radius=52.0864, center=<0 1.2901 0>, rotation=<rot 0.0018 -90 0.0489>}]
 main [{shapeType=Capsule, length=0.505, radius=192.2968, rotation=<rot 0 -90 0>}]
}
constraints [
 {bone1Name=JNT_A_Target, bone2Name=main, swing1LimitDegrees=45, swing2LimitDegrees=45, twistLimitDegrees=45}
 {bone1Name=JNT_B_Target, bone2Name=main, swing1LimitDegrees=45, swing2LimitDegrees=45, twistLimitDegrees=45}
]
